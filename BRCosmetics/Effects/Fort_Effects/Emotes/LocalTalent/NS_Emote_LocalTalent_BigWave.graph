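# Niagara burst effect for the BigWave part of the LocalTalent emote, built from three short-lived emitters.
# Ring spawns a warm orange-gold mesh that expands while its color flashes and fades and a material parameter animates.
# Sprites_group001 throws spinning, multicolor sprites with random visibility tags across four renderers, slowed by drag, distance-culled and scaled down in front-end preview.
# Glow adds a brief green sprite fading late in its life.
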NiagaraSystem /BRCosmetics/Effects/Fort_Effects/Emotes/LocalTalent/NS_Emote_LocalTalent_BigWave
dependencies {usedDataInterfaces [&NiagaraDataInterfaceCurve], usedDynamicInputs [], usedModules [&SystemState]}
emitters {
 Ring {
  emitterUpdateScript [
   &EmitterState(Inactive Response=::NewEnumerator1, Life Cycle Mode=::NewEnumerator1, Loop Behavior=::NewEnumerator1, Loop Duration=1, UseLoopDelay=false)
   &SpawnBurst_Instantaneous(Use Spawn Probability=false, UseLoopCountLimit=false)
  ]
  particleSpawnScript [
   &InitializeParticle(
    Color Mode=::NewEnumerator0
    Lifetime=0.5
    Mesh Scale=<5 1.5 1.5>
    Mesh Scale Mode=::NewEnumerator4
    UsePositionOffset=true
    Write Material Random=true
   )
   &InitialMeshOrientation(Mesh Orientation Mode=::NewEnumerator2, Use Rotation Vector=true)
  ]
  particleUpdateScript [
   &ParticleState()
   &ScaleMeshSize(Scale Factor=&VectorFromFloat(Value=&FloatFromCurve(
      FloatCurve=DataInterface() {propertyValues {Curve {keys [{arriveTangent=1.8728, interpMode=RCIM_Cubic, leaveTangent=1.8728, tangentMode=RCTM_User, time=0, value=0.5}, <at 1 1.5 RCIM_Cubic>]}}}
      Scale Curve=0.7
     )))
   /SetVariables(Particles.Color003:LinearColor=&MultiplyLinearColor(
     Linear Color=<rgba 1.2 0.7326 0.1114 1>
     Scale Factor=&MakeVector4_FromFloat(Float=&FloatFromCurve(
       FloatCurve=DataInterface() {propertyValues {Curve {keys [
           {arriveTangent=0.9513, interpMode=RCIM_Cubic, leaveTangent=0.9513, tangentMode=RCTM_User, time=0, value=0}
           {arriveTangent=10.1951, interpMode=RCIM_Cubic, leaveTangent=-5.7722, tangentMode=RCTM_Break, time=0.1, value=1.5}
           {arriveTangent=-1.8916, interpMode=RCIM_Cubic, leaveTangent=-1.8916, tangentMode=RCTM_User, time=0.3, value=0.575}
           {arriveTangent=-0.2689, interpMode=RCIM_Cubic, leaveTangent=-0.2689, tangentMode=RCTM_User, time=1, value=0}
          ]}}}
       Scale Curve=0.7
      ))
    )) {isSetParameters=true}
   &DynamicMaterialParameters(
    Index 0 Param 1=&FloatFromCurve(FloatCurve=DataInterface() {propertyValues {Curve {keys [
         {arriveTangent=-4.0585, interpMode=RCIM_Cubic, leaveTangent=-4.0585, tangentMode=RCTM_User, time=0, value=1}
         {arriveTangent=-1.0113, interpMode=RCIM_Cubic, leaveTangent=-1.0113, tangentMode=RCTM_User, time=1, value=-1.5}
        ]}}})
    Module.Index0Param0Write=true
    Module.Index0Param1Write=true
    Module.Index0Param2Write=true
    Module.Index0Param3Write=true
   )
  ]
  rendererClasses [&NiagaraMeshRendererProperties]
  renderers [{rendererClass=&NiagaraMeshRendererProperties, rendererIndex=0}]
 }
 Sprites_group001 {
  emitterUpdateScript [
   &EmitterState(
    Enable Distance Culling=true
    Life Cycle Mode=::NewEnumerator1
    Loop Behavior=::NewEnumerator1
    Loop Delay=&RandomRangeFloat(Minimum=0, Maximum=1)
    Loop Duration=1
    Max Distance Response=::KillImmediately
    Module.bUseMaxDistance=true
    Module.bUseMinDistance=false
    Scalability Mode=::NewEnumerator1
    UseLoopDelay=false
   )
   &SpawnBurst_Instantaneous(Spawn Count=15, Use Spawn Probability=false, UseLoopCountLimit=false)
  ]
  particleSpawnScript [
   &InitializeParticle(
    Color=&MakeCustomLinearColorFromBool(
     Bool=&RandomBool()
     True Color=<rgba 0 0.65 0.2256 1>
     False Color=&MakeCustomLinearColorFromBool(Bool=&RandomBool(), True Color=<rgba 0.7658 1 0 1>, False Color=<rgba 0.6979 0.1713 0 1>)
    )
    Lifetime Max=0.45
    Lifetime Min=0.2
    Lifetime Mode=::NewEnumerator1
    Sprite Rotation Mode=::NewEnumerator2
    Sprite Size Mode=::NewEnumerator1
    Uniform Sprite Size Max=8
    Uniform Sprite Size Min=6
    UsePositionOffset=false
    Write Material Random=true
   )
   /SetVariables(Particles.VisibilityTag=&RandomRangeInt(Minimum=0, Maximum=3)) {isSetParameters=true}
   /SetVariables(Particles.Color_triangle:LinearColor=&MakeCustomLinearColorFromBool(
     Bool=&RandomBool()
     True Color=<rgba 0 0.65 0.1934 1>
     False Color=&MakeCustomLinearColorFromBool(Bool=&RandomBool(), True Color=<rgba 0.4456 1 0 1>, False Color=<rgba 0 0.4818 0.7188 1>)
    )) {isSetParameters=true}
   /SetVariables(Particles.Color_triangleglow:LinearColor=&MultiplyLinearColor(
     Linear Color=$Particles.Color_triangle
     Scale Factor=&MakeVector4_FromLinearColor(LinearColor=/Niagara/DynamicInputs/TypeConversions/MakeLinearColorFromFloat(Value:NiagaraFloat=0.5))
    )) {isSetParameters=true}
   &ShapeLocation(MaskSpawnGroup=false, Override Seed=false, Sphere Radius=5, UseOwnerScale=true)
   &AddVelocity(Module.Use Falloff Distance=false, Override Seed=false, Velocity Mode=::NewEnumerator1, Velocity Speed=100)
  ]
  particleUpdateScript [
   &ParticleState()
   &SpriteRotationRate(Rotation Rate=&RandomRangeFloat(Minimum=-400, Maximum=400))
   &ScaleSpriteSize(Scale Factor=&Vector2DFromFloat(Value=&FloatFromCurve(
      FloatCurve=DataInterface() {propertyValues {Curve {keys [
          {interpMode=RCIM_Cubic, tangentMode=RCTM_User, time=0, value=0}
          {interpMode=RCIM_Cubic, tangentMode=RCTM_User, time=0.1, value=1}
          {arriveTangent=0.0233, interpMode=RCIM_Cubic, leaveTangent=0.0233, tangentMode=RCTM_User, time=0.7, value=1}
          {tangentMode=RCTM_User, time=1, value=0}
         ]}}}
      Scale Curve=&MakeCustomFloatFromBool(Bool=$User.bIsFrontEndPreview, True Float=0.6, False Float=1)
     )))
   /SetVariables(Particles.Size002:Vector2f=&Multiply_Vector2D(A=$Particles.SpriteSize, B=&Vector2DFromFloat(Value=1.5))) {isSetParameters=true}
   /SetVariables(Particles.Size_Triangle:Vector2f=&Multiply_Vector2D(A=$Particles.SpriteSize, B=&Vector2DFromFloat(Value=0.6))) {isSetParameters=true}
   &Drag(Drag=2, Use Linear Drag=true, Use Rotational Drag=true)
   &SolveForcesAndVelocity(Clamp Velocity=false, Limit Acceleration=false)
   &ScaleColor(
    Scale Alpha=&Multiply_Float(
     A=1
     B=&FloatFromCurve(FloatCurve=DataInterface() {propertyValues {Curve {keys [
          {interpMode=RCIM_Cubic, tangentMode=RCTM_User, time=0.1, value=0}
          {interpMode=RCIM_Cubic, tangentMode=RCTM_User, time=0.3, value=0.8}
          {tangentMode=RCTM_User, time=0.8, value=0}
         ]}}})
    )
    Scale RGB=&MakeVectorFromLinearColorRGB(Color=<rgba 1 1 1 1>)
    ScaleA=true
    ScaleRGB=true
   )
   /SetVariables(Particles.Color002=&MultiplyLinearColor(Linear Color=<rgba 1 1 1 1>, Scale Factor=&MakeVector4_FromLinearColor(LinearColor=$Particles.Color))) {isSetParameters=true}
  ]
  rendererClasses [&NiagaraSpriteRendererProperties]
  renderers [
   {rendererClass=&NiagaraSpriteRendererProperties}
   {rendererClass=&NiagaraSpriteRendererProperties, rendererIndex=1}
   {rendererClass=&NiagaraSpriteRendererProperties, rendererIndex=2}
   {rendererClass=&NiagaraSpriteRendererProperties, rendererIndex=3}
  ]
 }
 glow {
  emitterUpdateScript [
   &EmitterState(Inactive Response=::NewEnumerator1, Life Cycle Mode=::NewEnumerator1, Loop Behavior=::NewEnumerator1, Loop Duration=1, UseLoopDelay=false)
   &SpawnBurst_Instantaneous(Use Spawn Probability=false, UseLoopCountLimit=true)
  ]
  particleSpawnScript [
   &InitializeParticle(
    Color Mode=::NewEnumerator0
    Lifetime=0.3
    Sprite Size Mode=::NewEnumerator3
    Uniform Sprite Size=60
    UsePositionOffset=false
    Write Material Random=true
   )
   &DynamicMaterialParameters(
    Index 0 Param 1=4
    Index 0 Param 2=10
    Module.Index0Float0Write=true
    Module.Index0Float1Write=true
    Module.Index0Float2Write=false
    Module.Index0Float3Write=false
   )
  ]
  particleUpdateScript [
   &ParticleState()
   &ScaleColor(
    Scale Alpha=&FloatFromCurve(FloatCurve=DataInterface() {propertyValues {Curve {keys [<at 0 1>,<at 0.5 1>,<at 1 0>]}}}, Scale Curve=0.25)
    Scale RGB=&MakeVectorFromLinearColorRGB(Color=<rgba 0.0516 1 0.1313 1>)
    ScaleA=true
    ScaleRGB=true
   )
  ]
  rendererClasses [&NiagaraSpriteRendererProperties]
  renderers [{rendererClass=&NiagaraSpriteRendererProperties}]
 }
}
summary {
 systemName=NS_Emote_LocalTalent_BigWave
 emitters [
  {emitterName=Sprites_group001, rendererClasses [&NiagaraSpriteRendererProperties]}
  {emitterName=Ring, rendererClasses [&NiagaraMeshRendererProperties]}
  {emitterName=glow, rendererClasses [&NiagaraSpriteRendererProperties]}
 ]
 userVariables [
  User.bIsFrontEnd {type=NiagaraBool}
  User.bIsFrontEndPreview {type=NiagaraBool}
 ]
}
