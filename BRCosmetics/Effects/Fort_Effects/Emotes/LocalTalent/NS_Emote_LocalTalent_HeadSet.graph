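# Niagara system for the LocalTalent emote's headset effect, pairing repeating ray meshes with a soft green glow.
# Ray_Line continuously spawns short-lived meshes with randomized orientation, scaling them over life by a curve and boosting brightness, with separate alpha strengths for gameplay and front end.
# The glow emitter spawns sprites on a matching rate and loop, tinting them green, fading alpha by curve and feeding dynamic material parameters.
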NiagaraSystem /BRCosmetics/Effects/Fort_Effects/Emotes/LocalTalent/NS_Emote_LocalTalent_HeadSet
dependencies {
 usedDataInterfaces [&NiagaraDataInterfaceCurve]
 usedDynamicInputs [&RandomRangeFloat]
 usedModules [&SystemState]
}
emitters {
 Ray_Line {
  emitterUpdateScript [
   &EmitterState(Inactive Response=::NewEnumerator1, Life Cycle Mode=::NewEnumerator1, Loop Duration=7, UseLoopDelay=false)
   &SpawnRate(SpawnRate=2, Use Spawn Probability=false)
   &SpawnBurst_Instantaneous(Use Spawn Probability=true, UseLoopCountLimit=true)
  ]
  particleSpawnScript [
   &InitializeParticle(
    Lifetime=0.6
    Mesh Scale=<0.7 1.3 1.3>
    Mesh Scale Mode=::NewEnumerator4
    Position Offset=<0 0 -2>
    UsePositionOffset=true
    Write Material Random=true
   )
   &InitialMeshOrientation(Mesh Orientation Mode=::NewEnumerator2, Rotation=<0 -0.25 0>, Use Rotation Vector=true)
   &InitialMeshOrientation(
    Mesh Orientation Mode=::NewEnumerator2
    Rotation=&RandomRangeVector(Minimum=<0 0 0>, Maximum=<0 0 1>)
    Use Rotation Vector=true
   ) {module=InitialMeshOrientation001}
  ]
  particleUpdateScript [
   &ParticleState()
   &UpdateMeshOrientation()
   &ScaleMeshSize(Scale Factor=&VectorFromFloat(Value=&FloatFromCurve(
      FloatCurve=DataInterface() {propertyValues {Curve {keys [{arriveTangent=0.915, interpMode=RCIM_Cubic, leaveTangent=0.915, tangentMode=RCTM_User, time=0, value=0.6}, <at 1 1 RCIM_Cubic>]}}}
      Scale Curve=0.22
     )))
   &ScaleColor(
    Scale Alpha=&FN_FrontEndFloatModifier(Gameplay=25, Front End=18)
    Scale RGB=&Multiply_VectorByFloat(Vector=<1 1 1>, Float=10)
    ScaleA=true
    ScaleRGB=true
   )
  ]
  rendererClasses [&NiagaraMeshRendererProperties]
  renderers [{rendererClass=&NiagaraMeshRendererProperties, rendererIndex=0}]
 }
 glow {
  emitterUpdateScript [
   &EmitterState(Inactive Response=::NewEnumerator1, Life Cycle Mode=::NewEnumerator1, Loop Duration=7, UseLoopDelay=false)
   &SpawnRate(SpawnRate=2, Use Spawn Probability=false)
   &SpawnBurst_Instantaneous(Use Spawn Probability=false, UseLoopCountLimit=true)
  ]
  particleSpawnScript [
   &InitializeParticle(
    Color Mode=::NewEnumerator0
    Lifetime=0.5
    Sprite Size Mode=::NewEnumerator3
    Uniform Sprite Size=65
    UsePositionOffset=false
    Write Material Random=true
   )
   &DynamicMaterialParameters(
    Index 0 Param 1=4
    Index 0 Param 2=10
    Module.Index0Float0Write=true
    Module.Index0Float1Write=true
    Module.Index0Float2Write=false
    Module.Index0Float3Write=false
   )
  ]
  particleUpdateScript [
   &ParticleState()
   &ScaleColor(
    Scale Alpha=&FloatFromCurve(FloatCurve=DataInterface() {propertyValues {Curve {keys [<at 0 1>,<at 0.5 1>,<at 1 0>]}}}, Scale Curve=0.15)
    Scale RGB=&MakeVectorFromLinearColorRGB(Color=<rgba 0.0516 1 0.1313 1>)
    ScaleA=true
    ScaleRGB=true
   )
  ]
  rendererClasses [&NiagaraSpriteRendererProperties]
  renderers [{rendererClass=&NiagaraSpriteRendererProperties}]
 }
}
summary {
 systemName=NS_Emote_LocalTalent_HeadSet
 emitters [
  {emitterName=Ray_Line, rendererClasses [&NiagaraMeshRendererProperties]}
  {emitterName=glow, rendererClasses [&NiagaraSpriteRendererProperties]}
 ]
 userVariables [
  User.bIsFrontEnd {type=NiagaraBool}
  User.bIsFrontEndPreview {type=NiagaraBool}
 ]
}
