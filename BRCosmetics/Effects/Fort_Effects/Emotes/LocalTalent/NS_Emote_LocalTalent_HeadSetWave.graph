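# Niagara system for the LocalTalent emote, playing a short burst that, by its name, is a headset wave effect.
# Mesh emitters form an expanding wave, ray lines, a star and a star glow, alongside a soft glow sprite and a spray of small, randomly colored spinning sprites.
# Curves animate scale, color and dynamic material parameters, with some layers adjusting brightness for front end versus gameplay and time of day.
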NiagaraSystem /BRCosmetics/Effects/Fort_Effects/Emotes/LocalTalent/NS_Emote_LocalTalent_HeadSetWave
dependencies {
 usedDataInterfaces [&NiagaraDataInterfaceCurve, &NiagaraDataInterfaceColorCurve]
 usedDynamicInputs []
 usedModules [
  &self:SystemSpawnScript.SystemScriptSource.SystemScriptGraph.NiagaraNodeAssignment_0.SetVariables_#1
  &SystemState
  &InterpolateOverTime
 ]
}
emitters {
 Ray_Line {
  emitterUpdateScript [
   &EmitterState(Inactive Response=::NewEnumerator1, Life Cycle Mode=::NewEnumerator1, Loop Behavior=::NewEnumerator1, Loop Duration=1, UseLoopDelay=false)
   &SpawnBurst_Instantaneous(Use Spawn Probability=false, UseLoopCountLimit=false)
  ]
  particleSpawnScript [
   &InitializeParticle(
    Lifetime=0.55
    Mesh Scale=<0.8 1.3 1.3>
    Mesh Scale Mode=::NewEnumerator4
    Position Offset=<0 0 -3>
    UsePositionOffset=true
    Write Material Random=true
   )
   &InitialMeshOrientation(Mesh Orientation Mode=::NewEnumerator2, Rotation=<0 -0.25 0>, Use Rotation Vector=true)
   &InitialMeshOrientation(
    Mesh Orientation Mode=::NewEnumerator2
    Rotation=&RandomRangeVector(Minimum=<0 0 0>, Maximum=<0 0 1>)
    Use Rotation Vector=true
   ) {module=InitialMeshOrientation001}
  ]
  particleUpdateScript [
   &ParticleState()
   &UpdateMeshOrientation()
   &ScaleMeshSize(Scale Factor=&VectorFromFloat(Value=&FloatFromCurve(
      FloatCurve=DataInterface() {propertyValues {Curve {keys [{arriveTangent=0.915, interpMode=RCIM_Cubic, leaveTangent=0.915, tangentMode=RCTM_User, time=0, value=0.6}, <at 1 1 RCIM_Cubic>]}}}
      Scale Curve=0.23
     )))
   &ScaleColor(
    Scale Alpha=&Multiply_Float(
     A=&FN_FrontEndFloatModifier(Gameplay=30, Front End=20)
     B=&FNTimeOfDayScalar(Night Scale=0.8, Morning Scale=0.8, Evening Scale=0.8)
    )
    Scale RGB=&Multiply_VectorByFloat(Vector=<1 1 1>, Float=10)
    ScaleA=true
    ScaleRGB=true
   )
  ]
  rendererClasses [&NiagaraMeshRendererProperties]
  renderers [{rendererClass=&NiagaraMeshRendererProperties, rendererIndex=0}]
 }
 Sprites001 {
  emitterUpdateScript [
   &EmitterState(
    Enable Distance Culling=true
    Inactive Response=::NewEnumerator1
    Life Cycle Mode=::NewEnumerator1
    Loop Behavior=::NewEnumerator1
    Loop Duration=1
    Max Distance Response=::KillImmediately
    Module.bUseMaxDistance=true
    Module.bUseMinDistance=false
    Scalability Mode=::NewEnumerator1
    UseLoopDelay=false
   )
   &SpawnBurst_Instantaneous(
    Spawn Count=&RandomRangeInt(Minimum=10, Maximum=15)
    Use Spawn Probability=false
    UseLoopCountLimit=false
   )
  ]
  particleSpawnScript [
   &InitializeParticle(
    Color=&MakeCustomLinearColorFromBool(
     Bool=&RandomBool()
     True Color=<rgba 0 0.65 0.2256 1>
     False Color=&MakeCustomLinearColorFromBool(Bool=&RandomBool(), True Color=<rgba 0.7658 1 0 1>, False Color=<rgba 0.6979 0.1713 0 1>)
    )
    Lifetime Max=0.4
    Lifetime Min=0.25
    Lifetime Mode=::NewEnumerator1
    Sprite Rotation Mode=::NewEnumerator2
    Sprite Size Mode=::NewEnumerator1
    Uniform Sprite Size Min=8
    UsePositionOffset=false
    Write Material Random=true
   )
   &ShapeLocation(
    MaskSpawnGroup=false
    Non Uniform Scale=<1 1 0.5>
    Offset Mode=::NewEnumerator0
    Override Seed=false
    Rotation Axis=<1 0.5 0.4>
    Rotation Mode=::NewEnumerator1
    Sphere Radius=5
    UseOwnerScale=true
   )
   &AddVelocity(
    Cone Angle=70
    Cone Axis=<0 0 1>
    Module.UseConeFalloff=false
    Module.UseVelDistribution=false
    Override Seed=false
    Velocity Mode=::NewEnumerator2
    Velocity Speed=&RandomRangeFloat(Minimum=50, Maximum=120)
   )
   /SetVariables(Particles.VisibilityTag=&RandomRangeInt(Minimum=0, Maximum=3)) {isSetParameters=true}
  ]
  particleUpdateScript [
   &ParticleState()
   &SpriteRotationRate(Rotation Rate=&RandomRangeFloat(Minimum=-200, Maximum=200))
   &ScaleSpriteSize(Scale Factor=&Vector2DFromFloat(Value=&FloatFromCurve(FloatCurve=DataInterface() {propertyValues {Curve {keys [
          {interpMode=RCIM_Cubic, tangentMode=RCTM_User, time=0, value=1}
          {arriveTangent=-0.0231, interpMode=RCIM_Cubic, leaveTangent=-0.0231, tangentMode=RCTM_User, time=0.7, value=1}
          {tangentMode=RCTM_User, time=1, value=0}
         ]}}})))
   /SetVariables(Particles.Size002:Vector2f=&Multiply_Vector2D(A=$Particles.SpriteSize, B=&Vector2DFromFloat(Value=1.5))) {isSetParameters=true}
   /SetVariables(Particles.Size_Triangle:Vector2f=&Multiply_Vector2D(A=$Particles.SpriteSize, B=&Vector2DFromFloat(Value=0.55))) {isSetParameters=true}
   &Drag(Use Linear Drag=true, Use Rotational Drag=true)
   &SolveForcesAndVelocity(Clamp Velocity=false, Limit Acceleration=false)
   &ScaleColor(
    Scale Alpha=&Multiply_Float(
     A=&FNTimeOfDayScalar(Night Scale=1, Morning Scale=0.3, Evening Scale=0.3)
     B=&FloatFromCurve(FloatCurve=DataInterface() {propertyValues {Curve {keys [
          {interpMode=RCIM_Cubic, tangentMode=RCTM_User, time=0.1, value=0}
          {interpMode=RCIM_Cubic, tangentMode=RCTM_User, time=0.3, value=0.8}
          {tangentMode=RCTM_User, time=0.8, value=0}
         ]}}})
    )
    Scale RGB=&MakeVectorFromLinearColorRGB(Color=<rgba 1 1 1 1>)
    ScaleA=true
    ScaleRGB=true
   )
   /SetVariables(Particles.Color002=&MultiplyLinearColor(
     Linear Color=&MakeLinearColorFromVectorAndFloat(
      "Vector (RGB)"=&MakeVectorFromLinearColorRGB(Color=<rgba 1 1 1 1>)
      "Float (Alpha)"=&FN_FrontEndFloatModifier(Gameplay=1.2, Front End=0.5)
     )
     Scale Factor=&MakeVector4_FromLinearColor(LinearColor=$Particles.Color)
    )) {isSetParameters=true}
   /SetVariables(Particles.Color_triangle:LinearColor=&MakeLinearColorFromVectorAndFloat(
     "Vector (RGB)"=&MakeVectorFromLinearColorRGB(Color=&MakeCustomLinearColorFromBool(
       Bool=&RandomBool()
       True Color=<rgba 0 0.65 0.1934 1>
       False Color=&MakeCustomLinearColorFromBool(Bool=&RandomBool(), True Color=<rgba 0.4456 1 0 1>, False Color=<rgba 0 0.4818 0.7188 1>)
      ))
     "Float (Alpha)"=1
    )) {isSetParameters=true}
   /SetVariables(Particles.Color_triangleglow:LinearColor=&MultiplyLinearColor(
     Linear Color=$Particles.Color_triangle
     Scale Factor=&MakeVector4_FromLinearColor(LinearColor=/Niagara/DynamicInputs/TypeConversions/MakeLinearColorFromFloat(Value:NiagaraFloat=0.5))
    )) {isSetParameters=true}
  ]
  rendererClasses [&NiagaraSpriteRendererProperties]
  renderers [
   {rendererClass=&NiagaraSpriteRendererProperties}
   {rendererClass=&NiagaraSpriteRendererProperties, rendererIndex=1}
   {rendererClass=&NiagaraSpriteRendererProperties, rendererIndex=2}
   {rendererClass=&NiagaraSpriteRendererProperties, rendererIndex=3}
  ]
 }
 Star_Glowmesh {
  emitterUpdateScript [
   &EmitterState(Inactive Response=::NewEnumerator1, Life Cycle Mode=::NewEnumerator1, Loop Behavior=::NewEnumerator1, Loop Duration=1, UseLoopDelay=false)
   &SpawnBurst_Instantaneous(Use Spawn Probability=false, UseLoopCountLimit=true)
  ]
  particleSpawnScript [
   &InitializeParticle(
    Lifetime=0.4
    Mesh Scale=<0.5 0.5 2>
    Mesh Scale Mode=::NewEnumerator4
    Position Offset=<0 0 -6>
    Sprite Size Mode=::NewEnumerator3
    Uniform Sprite Size=23
    UsePositionOffset=true
    Write Material Random=true
   )
   &InitialMeshOrientation(Mesh Orientation Mode=::NewEnumerator2, Use Rotation Vector=true)
  ]
  particleUpdateScript [
   &ParticleState()
   &ScaleMeshSize(Scale Factor=&MakeVector(
     X=1
     Y=1
     Z=&FloatFromCurve(FloatCurve=DataInterface() {propertyValues {Curve {keys [<at 0 1>,<at 1 1.2>]}}})
    ))
   &ScaleColor(
    Scale Alpha=&FloatFromCurve(
     FloatCurve=DataInterface() {propertyValues {Curve {keys [<at 0 0>,<at 0.2 1>,<at 0.6 1>,<at 1 0>]}}}
     Scale Curve=&Multiply_Float(
      A=&FN_FrontEndFloatModifier(Gameplay=1, Front End=0.5)
      B=&FNTimeOfDayScalar(Night Scale=0.8, Morning Scale=0.8, Evening Scale=0.8)
     )
    )
    Scale RGB=&Multiply_VectorByFloat(
     Vector=&MakeVectorFromLinearColorRGB(Color=&ColorFromCurve(ColorCurve=DataInterface() {propertyValues {
         AlphaCurve {keys []}
         BlueCurve {keys [<at 0.1761 0 RCIM_Cubic>,<at 0.2813 0 RCIM_Cubic>,<at 0.6291 0>,<at 0.7781 1>]}
         GreenCurve {keys [
           <at 0.1761 0.6134 RCIM_Cubic>
           {arriveTangent=0.8533, interpMode=RCIM_Cubic, leaveTangent=0.8533, time=0.2813, value=1}
           <at 0.6291 1>
           <at 0.7781 0.4312>
          ]}
         RedCurve {keys [
           <at 0.1761 1 RCIM_Cubic>
           {arriveTangent=-2.1217, interpMode=RCIM_Cubic, leaveTangent=-2.1217, time=0.2813, value=0.0389}
           <at 0.6291 0.0389>
           <at 0.7781 0>
          ]}
        }}))
     Float=0.6
    )
    ScaleA=true
    ScaleRGB=true
   )
   &SolveForcesAndVelocity(Clamp Velocity=false, Limit Acceleration=false)
  ]
  rendererClasses [&NiagaraMeshRendererProperties]
  renderers [{rendererClass=&NiagaraMeshRendererProperties, rendererIndex=0}]
 }
 Star_mesh {
  emitterUpdateScript [
   &EmitterState(Inactive Response=::NewEnumerator1, Life Cycle Mode=::NewEnumerator1, Loop Behavior=::NewEnumerator1, Loop Duration=1, UseLoopDelay=false)
   &SpawnBurst_Instantaneous(Use Spawn Probability=false, UseLoopCountLimit=true)
  ]
  particleSpawnScript [
   &InitializeParticle(
    Color Mode=::NewEnumerator0
    Lifetime=0.55
    Mesh Scale=<0.5 0.5 0.5>
    Mesh Scale Mode=::NewEnumerator4
    UsePositionOffset=true
    Write Material Random=true
   )
   &InitialMeshOrientation(Mesh Orientation Mode=::NewEnumerator2, Rotation=<0 -0.25 0>, Use Rotation Vector=true)
   &AddVelocity(Velocity=<0 0 15>)
  ]
  particleUpdateScript [
   &ParticleState()
   &ScaleMeshSize(Scale Factor=&VectorFromFloat(Value=&MakeFloatFromVector(
      Vector=&VectorFromFloat(Value=&FloatFromCurve(
        FloatCurve=DataInterface() {propertyValues {Curve {keys [{arriveTangent=1.8728, interpMode=RCIM_Cubic, leaveTangent=1.8728, tangentMode=RCTM_User, time=0, value=1}, <at 1 2 RCIM_Cubic>]}}}
        Scale Curve=1.3
       ))
      Channel=::NewEnumerator0
     )))
   /SetVariables(Particles.Color003:LinearColor=&MultiplyLinearColor(
     Linear Color=<rgba 1.2 0.7326 0.1114 1>
     Scale Factor=&MakeVector4_FromFloat(Float=&FloatFromCurve(
       FloatCurve=DataInterface() {propertyValues {Curve {keys [
           {arriveTangent=0.9513, interpMode=RCIM_Cubic, leaveTangent=0.9513, tangentMode=RCTM_User, time=0, value=0}
           {arriveTangent=10.1951, interpMode=RCIM_Cubic, leaveTangent=-5.7722, tangentMode=RCTM_Break, time=0.1, value=1.5}
           {arriveTangent=-1.8916, interpMode=RCIM_Cubic, leaveTangent=-1.8916, tangentMode=RCTM_User, time=0.3, value=0.575}
           {arriveTangent=-0.2689, interpMode=RCIM_Cubic, leaveTangent=-0.2689, tangentMode=RCTM_User, time=1, value=0}
          ]}}}
       Scale Curve=&FN_FrontEndFloatModifier(Gameplay=1.5, Front End=1)
      ))
    )) {isSetParameters=true}
   &DynamicMaterialParameters(
    Index 0 Param 1=&FloatFromCurve(FloatCurve=DataInterface() {propertyValues {Curve {keys [
         {arriveTangent=-4.0585, interpMode=RCIM_Cubic, leaveTangent=-4.0585, tangentMode=RCTM_User, time=0, value=1}
         {arriveTangent=-1.0113, interpMode=RCIM_Cubic, leaveTangent=-1.0113, tangentMode=RCTM_User, time=1, value=-1.5}
        ]}}})
    Module.Index0Param0Write=true
    Module.Index0Param1Write=true
    Module.Index0Param2Write=true
    Module.Index0Param3Write=true
   )
   &SolveForcesAndVelocity(Clamp Velocity=false, Limit Acceleration=false)
  ]
  rendererClasses [&NiagaraMeshRendererProperties]
  renderers [{rendererClass=&NiagaraMeshRendererProperties, rendererIndex=0}]
 }
 Wave001 {
  emitterUpdateScript [
   &EmitterState(Inactive Response=::NewEnumerator1, Life Cycle Mode=::NewEnumerator1, Loop Behavior=::NewEnumerator1, Loop Duration=1, UseLoopDelay=false)
   &SpawnBurst_Instantaneous(Use Spawn Probability=false, UseLoopCountLimit=false)
  ]
  particleSpawnScript [
   &InitializeParticle(
    Color Mode=::NewEnumerator0
    Lifetime Max=0.44
    Lifetime Min=0.33
    Lifetime Mode=::NewEnumerator1
    Mesh Scale=<0.8 0.8 0.65>
    Mesh Scale Mode=::NewEnumerator4
    Position Offset=<0 0 -2>
    UsePositionOffset=true
    Write Material Random=true
   )
   &InitialMeshOrientation(Mesh Orientation Mode=::NewEnumerator2, Use Rotation Vector=true)
   &AddVelocity(Velocity=<0 0 15>)
   &CameraOffset(Camera Offset Amount=-3)
  ]
  particleUpdateScript [
   &ParticleState()
   &ScaleMeshSize(Scale Factor=&VectorFromFloat(Value=&FloatFromCurve(
      FloatCurve=DataInterface() {propertyValues {Curve {keys [
          {arriveTangent=2.5611, interpMode=RCIM_Cubic, leaveTangent=2.5611, tangentMode=RCTM_User, time=0, value=0}
          {arriveTangent=0.8945, interpMode=RCIM_Cubic, leaveTangent=0.8945, tangentMode=RCTM_User, time=0.2, value=0.75}
          {arriveTangent=0.2567, interpMode=RCIM_Cubic, leaveTangent=0.2567, tangentMode=RCTM_User, time=1, value=1.2}
         ]}}}
      Scale Curve=1.1
     )))
   &DynamicMaterialParameters(
    Index 0 Param 1=1
    Index 0 Param 2=0.08
    Index 0 Param 3=&FloatFromCurve(
     FloatCurve=DataInterface() {propertyValues {Curve {keys [<at 0 0.55>, {arriveTangent=0.8333, time=0.1, value=1}, {arriveTangent=-0.875, leaveTangent=-0.875, time=0.5, value=1}, <at 1 0.4>]}}}
     Scale Curve=1.5
    )
    Module.Index0Float0Write=true
    Module.Index0Float1Write=true
    Module.Index0Float2Write=true
    Module.Index0Float3Write=false
   )
   &ScaleColor(
    Scale Alpha=&FloatFromCurve(FloatCurve=DataInterface() {propertyValues {Curve {keys [<at 0 1>,<at 0.5 1>,<at 1 0>]}}})
    Scale RGB=&MakeVectorFromLinearColorRGB(Color=<rgba 0.0247 0.2969 0.1189 1>)
    ScaleA=true
    ScaleRGB=true
   )
   /SetVariables(Particles.Color002=&MakeLinearColorFromVectorAndFloat(
     "Vector (RGB)"=&Multiply_VectorByFloat(Vector=&MakeVectorFromLinearColorRGB(Color=<rgba 2.6055 5 0.4975 1>), Float=1)
     "Float (Alpha)"=&FloatFromCurve(FloatCurve=DataInterface() {propertyValues {Curve {keys [<at 0.2 1.5>,<at 0.4 0.525>,<at 1 0>]}}}, Scale Curve=2)
    )) {isSetParameters=true}
   /SetVariables(Particles.scale_small:Vector3f=&Multiply_VectorByFloat(Vector=$Particles.Scale, Float=0.75)) {isSetParameters=true}
   /SetVariables(Particles.Color_light02:LinearColor=&MakeLinearColorFromVectorAndFloat(
     "Vector (RGB)"=&Multiply_VectorByFloat(Vector=&MakeVectorFromLinearColorRGB(Color=<rgba 0.6235 1 0 1>), Float=2)
     "Float (Alpha)"=&FloatFromCurve(FloatCurve=DataInterface() {propertyValues {Curve {keys [<at 0.2 1.5>,<at 0.5 0.6>,<at 1 0>]}}})
    )) {isSetParameters=true}
   &Drag(Drag=0.2, Use Linear Drag=true, Use Rotational Drag=true) {enabled=false}
   &SolveForcesAndVelocity(Clamp Velocity=false, Limit Acceleration=false)
  ]
  rendererClasses [&NiagaraMeshRendererProperties]
  renderers [
   {rendererClass=&NiagaraMeshRendererProperties, rendererIndex=0}
   {rendererClass=&NiagaraMeshRendererProperties, rendererIndex=1}
   {rendererClass=&NiagaraMeshRendererProperties, rendererIndex=2}
   {rendererClass=&NiagaraMeshRendererProperties, rendererIndex=3}
  ]
 }
 glow {
  emitterUpdateScript [
   &EmitterState(Inactive Response=::NewEnumerator1, Life Cycle Mode=::NewEnumerator1, Loop Behavior=::NewEnumerator1, Loop Duration=1, UseLoopDelay=false)
   &SpawnBurst_Instantaneous(Use Spawn Probability=false, UseLoopCountLimit=true)
  ]
  particleSpawnScript [
   &InitializeParticle(
    Color Mode=::NewEnumerator0
    Lifetime=0.6
    Position Offset=<0 0 -1>
    Sprite Size Mode=::NewEnumerator3
    Uniform Sprite Size=70
    UsePositionOffset=true
    Write Material Random=true
   )
   &DynamicMaterialParameters(
    Index 0 Param 1=4
    Index 0 Param 2=10
    Module.Index0Float0Write=true
    Module.Index0Float1Write=true
    Module.Index0Float2Write=false
    Module.Index0Float3Write=false
   )
  ]
  particleUpdateScript [
   &ParticleState()
   &ScaleColor(
    Scale Alpha=&FloatFromCurve(FloatCurve=DataInterface() {propertyValues {Curve {keys [<at 0 1>,<at 0.5 1>,<at 1 0>]}}}, Scale Curve=0.2)
    Scale RGB=&MakeVectorFromLinearColorRGB(Color=<rgba 0.0516 1 0.1313 1>)
    ScaleA=true
    ScaleRGB=true
   )
  ]
  rendererClasses [&NiagaraSpriteRendererProperties]
  renderers [{rendererClass=&NiagaraSpriteRendererProperties}]
 }
}
summary {
 systemName=NS_Emote_LocalTalent_HeadSetWave
 emitters [
  {emitterName=Wave001, rendererClasses [&NiagaraMeshRendererProperties]}
  {emitterName=Ray_Line, rendererClasses [&NiagaraMeshRendererProperties]}
  {emitterName=Star_mesh, rendererClasses [&NiagaraMeshRendererProperties]}
  {emitterName=Star_Glowmesh, rendererClasses [&NiagaraMeshRendererProperties]}
  {emitterName=Sprites001, rendererClasses [&NiagaraSpriteRendererProperties]}
  {emitterName=glow, rendererClasses [&NiagaraSpriteRendererProperties]}
 ]
 userVariables [
  User.bIsFrontEnd {type=NiagaraBool}
  User.bIsFrontEndPreview {type=NiagaraBool}
 ]
}
